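# Idle glow effect for the IndoorLace mic stand, built as a Niagara system with one emitter, BaseGlows_MicGlow.
# A burst spawns two persistent, warm-tinted particles, partitioned for size and camera offset, feeding two sprite renderers and a mesh renderer.
# System reactive and smoothed interpolation values drive scale and brightness, while waveforms add subtle halo pulsing.
# A time-of-day curve modulates alpha, with a lower curve scale in front-end preview.
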NiagaraSystem /SparksCosmetics/Instrument/MicStand/IndoorLace/Effects/Mic_Idle/NS_Mic_IndoorLace_Idle_FX
dependencies {
 usedDataInterfaces [&NiagaraDataInterfaceCurve, &NiagaraDataInterfaceColorCurve, &NiagaraDataInterfaceArrayFloat]
 usedDynamicInputs [
  &ColorFromCurve
  &MakeFloatFromVector
  &SetBoolByFloatComparison
 ]
 usedModules [
  &SystemState
  &self:SystemSpawnScript.SystemScriptSource.SystemScriptGraph.NiagaraNodeAssignment_1.SetVariables_#1
  &self:SystemSpawnScript.SystemScriptSource.SystemScriptGraph.NiagaraNodeAssignment_2.SetVariables_#2
  &Timeline
  &self:SystemSpawnScript.SystemScriptSource.SystemScriptGraph.NiagaraNodeAssignment_3.SetVariables_#3
  &self:SystemSpawnScript.SystemScriptSource.SystemScriptGraph.NiagaraNodeAssignment_5.SetVariables_#4
  &InterpolateOverTime
  &self:SystemSpawnScript.SystemScriptSource.SystemScriptGraph.NiagaraNodeAssignment_4.SetVariables_#5
 ]
}
emitters {BaseGlows_MicGlow {
  emitterUpdateScript [
   &EmitterState()
   &SpawnBurst_Instantaneous(Spawn Count=2, Spawn Group=1, Use Spawn Probability=true, UseLoopCountLimit=true) {module=SpawnBurst_Instantaneous001}
  ]
  particleSpawnScript [
   &PartitionParticles()
   /SetVariables(Particles.VisibilityTag=$Particles.PartitionParticles.Partition) {isSetParameters=true}
   &InitializeParticle(
    Color=&MergeHelper_ColorSwitch(0=<rgba 1 0.8663 0.5041 0.72>, 1=<rgba 1 0.8663 0.5041 0.72>, 2=?, 3=?, Index=$Particles.PartitionParticles.Partition, Num=::NewEnumerator0)
    Position Offset=<-3 0 0>
    Sprite Size Mode=::NewEnumerator3
    Uniform Sprite Size=&MergeHelper_FloatSwitch(0=20, 1=18, 2=?, 3=?, Index=$Particles.PartitionParticles.Partition, Num=::NewEnumerator0)
    UsePositionOffset=true
    Write Material Random=true
   )
   &CameraOffset(Camera Offset Amount=&MergeHelper_FloatSwitch(0=0, 1=10, 2=?, 3=?, Index=$Particles.PartitionParticles.Partition, Num=::NewEnumerator0))
   &FN_Sprite_Attributes(
    Camera Offset=15
    Color=&MakeLinearColorFromVectorAndFloat("Vector (RGB)"=&MakeVectorFromLinearColorRGB(Color=$Particles.Color), "Float (Alpha)"=0.12)
    Create Camera Offset Attribute=true
    Create Color Attribute=true
    Create SpriteSize Attribute=true
    SpriteSize=&Multiply_Vector2D_ByFloat(Float=2.25, Vector2D=$Particles.SpriteSize)
    Write to Module Namespace=true
   )
   /SetVariables(Particles.HaloDynamic:Vector4f=&MakeVector4(X=1, Y=0, Z=1, W=0)) {isSetParameters=true}
  ]
  particleUpdateScript [
   &ParticleState(Kill Particles When Lifetime Has Elapsed=false)
   &ScaleSpriteSize(
    Scale Sprite Size Mode=::NewEnumerator0
    Uniform Scale Factor=&Multiply_Float(A=$System.EngineOwnerScale, B=&Lerp_Float(A=0.8, B=1.3, Alpha=$System.reactiveFloat))
   )
   &ScaleColor(
    Scale Alpha=&Multiply_Float(
     A=&FloatFromCurve(
      FloatCurve=DataInterface() {propertyValues {Curve {keys [
          {interpMode=RCIM_Cubic, leaveTangentWeight=0.1001, tangentMode=RCTM_User, time=0, value=0.9}
          {arriveTangentWeight=0.1021, interpMode=RCIM_Cubic, leaveTangentWeight=0.0679, tangentMode=RCTM_User, time=0.3003, value=1}
          {arriveTangentWeight=0.0666, interpMode=RCIM_Cubic, leaveTangentWeight=0.0667, tangentMode=RCTM_User, time=0.5, value=1}
          {arriveTangent=-0.2, interpMode=RCIM_Cubic, leaveTangent=-0.2, time=0.7, value=0.95}
          {arriveTangentWeight=0.1, interpMode=RCIM_Cubic, tangentMode=RCTM_User, time=1, value=0.9}
         ]}}}
      CurveIndex=Linked() {linkedVariable {name=NPC.FortniteNPC.RemappedTimeOfDay_0-1, type=NiagaraFloat}}
      Scale Curve=&MakeCustomFloatFromBool(Bool=$User.bIsFrontEndPreview, True Float=0.7, False Float=1)
     )
     B=&Lerp_Float(A=0.75, B=1, Alpha=$System.InterpolateOverTime.InterpolateOverTime_Float.Moving Average)
    )
    Scale RGB=&Multiply_VectorByFloat(
     Vector=<1 1 1>
     Float=&Lerp_Float(A=1, B=1.2, Alpha=$System.InterpolateOverTime.InterpolateOverTime_Float.Moving Average)
    )
    ScaleA=true
    ScaleRGB=true
   )
   /SetVariables(
    Particles.FN_Sprite_Attributes.SpriteSize:Vector2f=&Multiply_Vector2D_ByFloat(
     Float=&Multiply_Float(
      A=$System.EngineOwnerScale
      B=&Lerp_Float(A=0.8, B=1.3, Alpha=$System.InterpolateOverTime.InterpolateOverTime_Float.Moving Average)
     )
     Vector2D=&Multiply_Vector2D_ByFloat(
      Float=&Waveform("Waveform [1]"=::NewEnumerator0, "[1] Phase Scale"=2.5, "[1] Frequency"=1, "[1] Amplitude Min/Max"=<0.98 1>)
      Vector2D=$Particles.Initial.FN_Sprite_Attributes.SpriteSize
     )
    )
    Particles.HaloColor:LinearColor=&Multiply_LinearColorByFloat(
     Linear Color=$Particles.FN_Sprite_Attributes.Color
     Float=&Multiply_Float(
      A=&Waveform("Waveform [1]"=::NewEnumerator12, "[1] Frequency"=5, "[1] Amplitude Min/Max"=<0.85 1>)
      B=&Lerp_Float(A=1, B=1.2, Alpha=$System.InterpolateOverTime.InterpolateOverTime_Float.Moving Average)
     )
    )
   ) {isSetParameters=true}
  ]
  rendererClasses [&NiagaraSpriteRendererProperties, &NiagaraMeshRendererProperties]
  renderers [
   {rendererClass=&NiagaraSpriteRendererProperties}
   {rendererClass=&NiagaraSpriteRendererProperties, rendererIndex=1}
   {rendererClass=&NiagaraMeshRendererProperties, rendererIndex=2}
  ]
 }}
summary {
 systemName=NS_Mic_IndoorLace_Idle_FX
 emitters [{
   emitterName=BaseGlows_MicGlow
   rendererClasses [&NiagaraSpriteRendererProperties, &NiagaraMeshRendererProperties]
  }]
 userVariables [
  User.bIsFrontEndPreview {type=NiagaraBool}
  User.IsInstrument {type=NiagaraFloat, defaultValue=0, description=""}
 ]
}
